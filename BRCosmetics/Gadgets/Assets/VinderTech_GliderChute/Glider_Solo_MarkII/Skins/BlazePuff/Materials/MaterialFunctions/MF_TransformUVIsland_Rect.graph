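# Material function for the BlazePuff glider skin that animates a rectangular UV island with translation, rotation and uniform scale.
# Starting offsets combine with timeline inputs, with translation driven by a normalized timeline over a UV range.
# UVs are offset, scaled around the island center and rotated through MF_Rotate2D.
# It outputs the element UV plus Step-based rectangle masks for the original and transformed positions, built from the island's min/max bounds.
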MaterialFunction /BRCosmetics/Gadgets/Assets/VinderTech_GliderChute/Glider_Solo_MarkII/Skins/BlazePuff/Materials/MaterialFunctions/MF_TransformUVIsland_Rect
graph {nodes [
  FunctionOutput(None=NamedRerouteUsage()) {outputName="Element UV (V2)"}
  FunctionOutput(None=NamedRerouteUsage()) {outputName="Orig Pos Mask (S)", sortPriority=20}
  FunctionOutput(None=NamedRerouteUsage()) {outputName="Transformed Pos Mask (S)", sortPriority=10}
  NamedRerouteDeclaration(None=FunctionInput() {bUsePreviewValueAsDefault=true, inputName=Element UV Min, inputType=FunctionInput_Vector2, sortPriority=10}) {name="UV Min (V2)", nodeColor=<rgba 0 0.9391 1 1>}
  n1 = FunctionInput() {bUsePreviewValueAsDefault=true, inputName=Element UV Max, inputType=FunctionInput_Vector2, sortPriority=20, previewValue=<1 1 0 1>}
  NamedRerouteDeclaration(None=@n1) {name="UV Max (V2)", nodeColor=<rgba 0 0.9391 1 1>}
  n2 = FunctionInput() {bUsePreviewValueAsDefault=true, inputName=Starting Scale, inputType=FunctionInput_Scalar, sortPriority=130, previewValue=<1 0 0 1>}
  NamedRerouteDeclaration(None=@n2) {name="Starting Scale (S)", nodeColor=<rgba 0 0.2087 1 1>}
  n3 = FunctionInput() {bUsePreviewValueAsDefault=true, inputName=Scale Timeline, inputType=FunctionInput_Scalar, sortPriority=410, previewValue=<1 0 0 1>}
  NamedRerouteDeclaration(None=@n3) {name="Scale Timeline (S)", nodeColor=<rgba 0.8174 1 0 1>}
  NamedRerouteDeclaration(None=Multiply(A=NamedRerouteUsage(), B=NamedRerouteUsage())) {name="Element Scale (S)", nodeColor=<rgba 1 0 0.3047 1>}
  n4 = Add(A=NamedRerouteUsage(), B=NamedRerouteUsage())
  NamedRerouteDeclaration(None=MaterialFunctionCall!(
    RotationAmount=NamedRerouteDeclaration(None=@n4) {name="Element Rotation (S)", nodeColor=<rgba 1 0 0.3047 1>}
    Coords=NamedRerouteUsage()
    Pivot=NamedRerouteUsage()
   ) {materialFunction=/BaseMaterial/Materials/Functions/MF_Rotate2D}.Result) {name="Element UV Out (V2)", nodeColor=<rgba 0.0522 1 0 1>}
  n5 = NamedRerouteUsage()
  n6 = NamedRerouteUsage()
  n7 = NamedRerouteDeclaration(None=Multiply(A=Add(A=@n5, B=@n6)) {constB=0.5}) {name="Element UV Center (V2)", nodeColor=<rgba 1 0 0.6797 1>}
  n8 = Subtract(A=NamedRerouteUsage(), B=@n7)
  n9 = NamedRerouteUsage()
  n10 = NamedRerouteUsage()
  NamedRerouteDeclaration(None=Add(A=Multiply(A=@n8, B=@n9), B=@n10)) {name="UV Min (Scaled) (V2)", nodeColor=<rgba 0.7109 1 0 1>}
  n11 = NamedRerouteUsage()
  n12 = NamedRerouteUsage()
  NamedRerouteDeclaration(None=Add(A=Multiply(A=@n11, B=@n12), B=NamedRerouteUsage())) {name="Element UV Offset (V2)", nodeColor=<rgba 1 0 0.4219 1>}
  n13 = Subtract(A=NamedRerouteUsage(), B=@n7)
  NamedRerouteDeclaration(None=Add(A=Multiply(A=@n13, B=@n9), B=@n10)) {name="UV Max (Scaled) (V2)", nodeColor=<rgba 0.7109 1 0 1>}
  n14 = NamedRerouteUsage()
  n15 = NamedRerouteUsage()
  MaterialExpressionMultiply_6 = Multiply(A=Step(Y=NamedRerouteUsage(), X=@n14), B=OneMinus(None=Step(Y=@n15, X=@n14)))
  n16 = Convert(None=@MaterialExpressionMultiply_6) {
   nodeName=Break Float2
   convertInputs [{
     type=Vector2
     expressionInput {
      expression=&self:MaterialExpressionMultiply_6
      inputName=None
      mask=0
      maskA=0
      maskB=0
      maskG=0
      maskR=0
      outputIndex=0
     }
    }]
   convertOutputs [{type=Scalar}, {type=Scalar}]
  }
  NamedRerouteDeclaration(None=Multiply(A=@n16, B=@n16)) {name="Element Orig Pos Mask (S)", nodeColor=<rgba 1 0 0.4219 1>}
  NamedRerouteDeclaration(None=FunctionInput(Preview=TextureCoordinate()) {bUsePreviewValueAsDefault=true, inputName=UV, inputType=FunctionInput_Vector2}) {name="UV (V2)", nodeColor=<rgba 0.0522 1 0 1>}
  n17 = NamedRerouteUsage()
  n18 = NamedRerouteUsage()
  n19 = Subtract(A=NamedRerouteDeclaration(None=Subtract(A=@n17, B=@n18)) {name="Element UV (V2)", nodeColor=<rgba 0 0.7422 1 1>}, B=NamedRerouteUsage())
  n20 = NamedRerouteUsage()
  NamedRerouteDeclaration(None=Add(A=Divide(A=@n19, B=@n20), B=NamedRerouteUsage())) {name="Element UV (Scaled) (V2)", nodeColor=<rgba 0.7109 1 0 1>}
  n21 = NamedRerouteUsage()
  n22 = NamedRerouteUsage()
  n23 = NamedRerouteUsage()
  n24 = Add(A=NamedRerouteUsage(), B=@n22)
  MaterialExpressionMultiply_3 = Multiply(A=Step(Y=Add(A=@n21, B=@n22), X=@n23), B=OneMinus(None=Step(Y=@n24, X=@n23)))
  n25 = Convert(None=@MaterialExpressionMultiply_3) {
   nodeName=Break Float2
   convertInputs [{
     type=Vector2
     expressionInput {
      expression=&self:MaterialExpressionMultiply_3
      inputName=None
      mask=0
      maskA=0
      maskB=0
      maskG=0
      maskR=0
      outputIndex=0
     }
    }]
   convertOutputs [{type=Scalar}, {type=Scalar}]
  }
  NamedRerouteDeclaration(None=Multiply(A=@n25, B=@n25)) {name="Element Transformed Pos Mask (S)", nodeColor=<rgba 1 0 0.4219 1>}
  NamedRerouteDeclaration(None=FunctionInput() {bUsePreviewValueAsDefault=true, inputName=Starting Pos UV Offset, inputType=FunctionInput_Vector2, sortPriority=110}) {name="Starting Pos UV Offset (V2)", nodeColor=<rgba 0 0.2087 1 1>}
  NamedRerouteDeclaration(None=FunctionInput() {bUsePreviewValueAsDefault=true, inputName=Translation UV Range, inputType=FunctionInput_Vector2, sortPriority=210}) {name="Translation UV Range (V2)", nodeColor=<rgba 1 0 0.1913 1>}
  n26 = FunctionInput() {bUsePreviewValueAsDefault=true, inputName="Translation Timeline (Normalized)", inputType=FunctionInput_Scalar, sortPriority=220}
  NamedRerouteDeclaration(None=@n26) {name="Translation Normalized Timeline (S)", nodeColor=<rgba 1 0 0.1913 1>}
  NamedRerouteDeclaration(None=FunctionInput() {bUsePreviewValueAsDefault=true, inputName=Starting Rotation Offset, inputType=FunctionInput_Scalar, sortPriority=120}) {name="Starting Rotation Offset (S)", nodeColor=<rgba 0 0.2087 1 1>}
  NamedRerouteDeclaration(None=FunctionInput() {bUsePreviewValueAsDefault=true, inputName=Rotation Timeline, inputType=FunctionInput_Scalar, sortPriority=310}) {name="Rotation Timeline (S)", nodeColor=<rgba 1 0.2435 0 1>}
 ]}
