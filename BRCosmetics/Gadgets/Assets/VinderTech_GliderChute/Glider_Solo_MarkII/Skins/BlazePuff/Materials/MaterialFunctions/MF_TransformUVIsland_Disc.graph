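# Material function for the BlazePuff glider skin that animates a circular UV island by translating, scaling and rotating it.
# Starting values combine with timeline inputs into offset, scale and rotation, applied to the input UVs via subtract, divide and MF_Rotate2D.
# Distance-and-step tests build hard disc masks for the original and transformed island positions, output alongside the transformed UV.
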MaterialFunction /BRCosmetics/Gadgets/Assets/VinderTech_GliderChute/Glider_Solo_MarkII/Skins/BlazePuff/Materials/MaterialFunctions/MF_TransformUVIsland_Disc
graph {nodes [
  FunctionOutput(None=NamedRerouteUsage()) {outputName="Element UV (V2)"}
  FunctionOutput(None=NamedRerouteUsage()) {outputName="Transformed Pos Mask (S)", sortPriority=10}
  FunctionOutput(None=NamedRerouteUsage()) {outputName="Orig Pos Mask (S)", sortPriority=20}
  n1 = FunctionInput() {bUsePreviewValueAsDefault=true, inputName=Starting Scale, inputType=FunctionInput_Scalar, sortPriority=130, previewValue=<1 0 0 1>}
  NamedRerouteDeclaration(None=@n1) {name="Starting Scale (S)", nodeColor=<rgba 0 0.2087 1 1>}
  n2 = FunctionInput() {bUsePreviewValueAsDefault=true, inputName=Scale Timeline, inputType=FunctionInput_Scalar, sortPriority=410, previewValue=<1 0 0 1>}
  NamedRerouteDeclaration(None=@n2) {name="Scale Timeline (S)", nodeColor=<rgba 0.8174 1 0 1>}
  NamedRerouteDeclaration(None=Multiply(A=NamedRerouteUsage(), B=NamedRerouteUsage())) {name="Element Scale (S)", nodeColor=<rgba 1 0 0.3047 1>}
  n3 = NamedRerouteUsage()
  n4 = NamedRerouteUsage()
  NamedRerouteDeclaration(None=Add(A=Multiply(A=@n3, B=@n4), B=NamedRerouteUsage())) {name="Element UV Offset (V2)", nodeColor=<rgba 1 0 0.4219 1>}
  n5 = Add(A=NamedRerouteUsage(), B=NamedRerouteUsage())
  NamedRerouteDeclaration(None=MaterialFunctionCall!(
    RotationAmount=NamedRerouteDeclaration(None=@n5) {name="Element Rotation (S)", nodeColor=<rgba 1 0 0.3047 1>}
    Coords=NamedRerouteUsage()
    Pivot=NamedRerouteUsage()
   ) {materialFunction=/BaseMaterial/Materials/Functions/MF_Rotate2D}.Result) {name="Element UV Out (V2)", nodeColor=<rgba 0.0522 1 0 1>}
  NamedRerouteDeclaration(None=FunctionInput(Preview=TextureCoordinate()) {bUsePreviewValueAsDefault=true, inputName=UV, inputType=FunctionInput_Vector2}) {name="UV (V2)", nodeColor=<rgba 0.0522 1 0 1>}
  NamedRerouteDeclaration(None=Multiply(A=NamedRerouteUsage(), B=NamedRerouteUsage())) {name="Element UV Radius (Scaled) (S)", nodeColor=<rgba 1 0 0.7031 1>}
  n6 = NamedRerouteUsage()
  n7 = NamedRerouteUsage()
  n8 = Subtract(A=NamedRerouteDeclaration(None=Subtract(A=@n6, B=@n7)) {name="Element UV (V2)", nodeColor=<rgba 0 0.7422 1 1>}, B=NamedRerouteUsage())
  n9 = NamedRerouteUsage()
  NamedRerouteDeclaration(None=Add(A=Divide(A=@n8, B=@n9), B=NamedRerouteUsage())) {name="Element UV (Scaled) (V2)", nodeColor=<rgba 0.7109 1 0 1>}
  n10 = FunctionInput() {bUsePreviewValueAsDefault=true, inputName=Element UV Radius, inputType=FunctionInput_Scalar, sortPriority=20, previewValue=<0.25 1 0 1>}
  NamedRerouteDeclaration(None=@n10) {name="Element UV Radius (S)", nodeColor=<rgba 0 0.9391 1 1>}
  n11 = NamedRerouteUsage()
  n12 = Length(None=Subtract(A=NamedRerouteUsage(), B=NamedRerouteUsage())) {is=class MaterialExpressionLength}
  NamedRerouteDeclaration(None=OneMinus(None=Step(Y=@n11, X=@n12))) {name="Element Orig Pos Mask (S)", nodeColor=<rgba 1 0 0.4219 1>}
  n13 = NamedRerouteUsage()
  n14 = NamedRerouteUsage()
  n15 = NamedRerouteUsage()
  n16 = Length(None=Subtract(A=NamedRerouteUsage(), B=Add(A=@n14, B=@n15)))
  NamedRerouteDeclaration(None=OneMinus(None=Step(Y=@n13, X=@n16))) {name="Element Transformed Pos Mask (S)", nodeColor=<rgba 1 0 0.4219 1>}
  NamedRerouteDeclaration(None=FunctionInput() {bUsePreviewValueAsDefault=true, inputName=Starting Pos UV Offset, inputType=FunctionInput_Vector2, sortPriority=110}) {name="Starting Pos UV Offset (V2)", nodeColor=<rgba 0 0.2087 1 1>}
  NamedRerouteDeclaration(None=FunctionInput() {bUsePreviewValueAsDefault=true, inputName=Translation UV Range, inputType=FunctionInput_Vector2, sortPriority=210}) {name="Translation UV Range (V2)", nodeColor=<rgba 1 0 0.1913 1>}
  n17 = FunctionInput() {bUsePreviewValueAsDefault=true, inputName="Translation Timeline (Normalized)", inputType=FunctionInput_Scalar, sortPriority=220}
  NamedRerouteDeclaration(None=@n17) {name="Translation Normalized Timeline (S)", nodeColor=<rgba 1 0 0.1913 1>}
  NamedRerouteDeclaration(None=FunctionInput() {bUsePreviewValueAsDefault=true, inputName=Starting Rotation Offset, inputType=FunctionInput_Scalar, sortPriority=120}) {name="Starting Rotation Offset (S)", nodeColor=<rgba 0 0.2087 1 1>}
  NamedRerouteDeclaration(None=FunctionInput() {bUsePreviewValueAsDefault=true, inputName=Rotation Timeline, inputType=FunctionInput_Scalar, sortPriority=310}) {name="Rotation Timeline (S)", nodeColor=<rgba 1 0.2435 0 1>}
  n18 = FunctionInput() {bUsePreviewValueAsDefault=true, inputName=Element UV Center, inputType=FunctionInput_Vector2, sortPriority=10, previewValue=<0.5 0.5 0 1>}
  NamedRerouteDeclaration(None=@n18) {name="Element UV Center (V2)", nodeColor=<rgba 0 0.9391 1 1>}
 ]}
