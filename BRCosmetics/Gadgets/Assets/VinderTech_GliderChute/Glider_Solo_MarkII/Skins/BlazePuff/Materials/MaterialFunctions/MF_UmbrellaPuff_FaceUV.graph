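# Material function for the BlazePuff umbrella glider skin that projects face-decal UVs with independently adjustable eyes and mouth.
# Builds base UVs per vertex from pre-skinned local XY using pivot, size and rotation.
# Moves each feature via MF_TransformUVIsland_Disc, blanks original spots, layers the moved UVs and saturates against wrapping.
# Also outputs a UV-disc and normal-Z projection mask, mesh-UV derivatives and debug masks.
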MaterialFunction /BRCosmetics/Gadgets/Assets/VinderTech_GliderChute/Glider_Solo_MarkII/Skins/BlazePuff/Materials/MaterialFunctions/MF_UmbrellaPuff_FaceUV
graph {expressions {
  MaterialExpressionAdd_4 {
   class=MaterialExpressionAdd
   inputs [
    A {from=MaterialExpressionMaterialFunctionCall_3, output=Result}
    B {from=""}
   ]
   properties {constB=0.5}
  }
  MaterialExpressionAppendVector_11 {
   class=MaterialExpressionAppendVector
   inputs [
    A {from=MaterialExpressionVectorParameter_9, output=Offset U}
    B {from=MaterialExpressionVectorParameter_9, output=Offset U}
   ]
  }
  MaterialExpressionAppendVector_12 {
   class=MaterialExpressionAppendVector
   inputs [
    A {from=MaterialExpressionVectorParameter_14, output=Offset U}
    B {from=MaterialExpressionVectorParameter_14, output=Offset U}
   ]
  }
  MaterialExpressionAppendVector_15 {
   class=MaterialExpressionAppendVector
   inputs [
    A {from=MaterialExpressionVectorParameter_18, output=Center U}
    B {from=MaterialExpressionVectorParameter_18, output=Center U}
   ]
  }
  MaterialExpressionAppendVector_2 {
   class=MaterialExpressionAppendVector
   inputs [
    A {from=MaterialExpressionVectorParameter_1, output=Center U}
    B {from=MaterialExpressionVectorParameter_1, output=Center U}
   ]
  }
  MaterialExpressionAppendVector_21 {
   class=MaterialExpressionAppendVector
   inputs [
    A {from=MaterialExpressionVectorParameter_23, output=Center U}
    B {from=MaterialExpressionVectorParameter_23, output=Center U}
   ]
  }
  MaterialExpressionAppendVector_22 {
   class=MaterialExpressionAppendVector
   inputs [
    A {from=MaterialExpressionVectorParameter_24, output=Center U}
    B {from=MaterialExpressionVectorParameter_24, output=Center U}
   ]
  }
  MaterialExpressionAppendVector_3 {
   class=MaterialExpressionAppendVector
   inputs [
    A {from=MaterialExpressionVectorParameter_2, output=Pivot X}
    B {from=MaterialExpressionMultiply_4, output=""}
   ]
  }
  MaterialExpressionAppendVector_8 {
   class=MaterialExpressionAppendVector
   inputs [
    A {from=MaterialExpressionVectorParameter_8, output=Offset U}
    B {from=MaterialExpressionVectorParameter_8, output=Offset U}
   ]
  }
  MaterialExpressionComponentMask_0 {class=MaterialExpressionComponentMask, inputs [None {from=MaterialExpressionVertexInterpolator_0, output=PS}], properties {b=true}}
  MaterialExpressionConstant2Vector_2 {class=MaterialExpressionConstant2Vector}
  MaterialExpressionConstant_0 {class=MaterialExpressionConstant}
  MaterialExpressionConstant_2 {class=MaterialExpressionConstant}
  MaterialExpressionConvert_0 {
   class=MaterialExpressionConvert
   inputs [
    None {from=MaterialExpressionNamedRerouteUsage_10}
    None {from=MaterialExpressionNamedRerouteUsage_9}
    None {from=MaterialExpressionNamedRerouteUsage_11}
   ]
   properties {
    nodeName=Make Vector3
    convertInputs [
     {
      type=Scalar
      expressionInput {
       expression=&self:MaterialExpressionNamedRerouteUsage_10
       inputName=None
       mask=0
       maskA=0
       maskB=0
       maskG=0
       maskR=0
       outputIndex=0
      }
     }
     {
      type=Scalar
      expressionInput {
       expression=&self:MaterialExpressionNamedRerouteUsage_9
       inputName=None
       mask=0
       maskA=0
       maskB=0
       maskG=0
       maskR=0
       outputIndex=0
      }
     }
     {
      type=Scalar
      expressionInput {
       expression=&self:MaterialExpressionNamedRerouteUsage_11
       inputName=None
       mask=0
       maskA=0
       maskB=0
       maskG=0
       maskR=0
       outputIndex=0
      }
     }
    ]
    convertOutputs [{type=Vector3}]
   }
  }
  MaterialExpressionConvert_1 {
   class=MaterialExpressionConvert
   inputs [
    None {from=MaterialExpressionNamedRerouteUsage_8}
    None {from=MaterialExpressionNamedRerouteUsage_6}
    None {from=MaterialExpressionNamedRerouteUsage_7}
   ]
   properties {
    nodeName=Make Vector3
    convertInputs [
     {
      type=Scalar
      expressionInput {
       expression=&self:MaterialExpressionNamedRerouteUsage_8
       inputName=None
       mask=0
       maskA=0
       maskB=0
       maskG=0
       maskR=0
       outputIndex=0
      }
     }
     {
      type=Scalar
      expressionInput {
       expression=&self:MaterialExpressionNamedRerouteUsage_6
       inputName=None
       mask=0
       maskA=0
       maskB=0
       maskG=0
       maskR=0
       outputIndex=0
      }
     }
     {
      type=Scalar
      expressionInput {
       expression=&self:MaterialExpressionNamedRerouteUsage_7
       inputName=None
       mask=0
       maskA=0
       maskB=0
       maskG=0
       maskR=0
       outputIndex=0
      }
     }
    ]
    convertOutputs [{type=Vector3}]
   }
  }
  MaterialExpressionDDX_0 {class=MaterialExpressionDDX, inputs [Value {from=MaterialExpressionTextureCoordinate_3, output=""}], properties {}}
  MaterialExpressionDDY_0 {class=MaterialExpressionDDY, inputs [Value {from=MaterialExpressionTextureCoordinate_3, output=""}], properties {}}
  MaterialExpressionDivide_1 {
   class=MaterialExpressionDivide
   inputs [
    A {from=MaterialExpressionLength_0}
    B {from=MaterialExpressionVectorParameter_1, output=Radius}
   ]
  }
  MaterialExpressionDivide_5 {
   class=MaterialExpressionDivide
   inputs [
    A {from=MaterialExpressionSubtract_8}
    B {from=MaterialExpressionReroute_0}
   ]
  }
  MaterialExpressionFunctionOutput_0 {
   class=MaterialExpressionFunctionOutput
   inputs [None {from=MaterialExpressionNamedRerouteUsage_2, output=""}]
   properties {outputName="Base Projected UV (V2)"}
  }
  MaterialExpressionFunctionOutput_1 {
   class=MaterialExpressionFunctionOutput
   inputs [None {from=MaterialExpressionNamedRerouteUsage_3, output=""}]
   properties {outputName="Transformed Projected UV (V2)", sortPriority=10}
  }
  MaterialExpressionFunctionOutput_2 {
   class=MaterialExpressionFunctionOutput
   inputs [None {from=MaterialExpressionNamedRerouteUsage_4, output=""}]
   properties {outputName="Projection Mask (S)", sortPriority=20}
  }
  MaterialExpressionFunctionOutput_3 {
   class=MaterialExpressionFunctionOutput
   inputs [None {from=MaterialExpressionNamedRerouteUsage_0, output=""}]
   properties {outputName="UV DDX (V2)", sortPriority=30}
  }
  MaterialExpressionFunctionOutput_4 {
   class=MaterialExpressionFunctionOutput
   inputs [None {from=MaterialExpressionNamedRerouteUsage_1, output=""}]
   properties {outputName="UV DDY (V2)", sortPriority=40}
  }
  MaterialExpressionFunctionOutput_5 {
   class=MaterialExpressionFunctionOutput
   inputs [None {from=MaterialExpressionNamedRerouteUsage_5, output=""}]
   properties {outputName="DEBUG (V3)", sortPriority=300}
  }
  MaterialExpressionFunctionOutput_6 {
   class=MaterialExpressionFunctionOutput
   inputs [None {from="", output=""}]
   properties {outputName=//----------- DEBUG -----------//, sortPriority=100}
  }
  MaterialExpressionFunctionOutput_7 {
   class=MaterialExpressionFunctionOutput
   inputs [None {from=MaterialExpressionConvert_0, output=""}]
   properties {outputName="Orig Pos Mask (V3)", sortPriority=110}
  }
  MaterialExpressionFunctionOutput_8 {
   class=MaterialExpressionFunctionOutput
   inputs [None {from=MaterialExpressionConvert_1, output=""}]
   properties {outputName="Transformed Pos Mask (V3)", sortPriority=120}
  }
  MaterialExpressionLength_0 {class=MaterialExpressionLength, inputs [None {from=MaterialExpressionSubtract_0, output=""}], properties {}}
  MaterialExpressionLinearInterpolate_0 {
   class=MaterialExpressionLinearInterpolate
   inputs [
    A {from=""}
    B {from=MaterialExpressionSmoothStep_0}
    Alpha {from=MaterialExpressionReroute_1}
   ]
   properties {constA=1}
  }
  MaterialExpressionLinearInterpolate_1 {
   class=MaterialExpressionLinearInterpolate
   inputs [
    A {from=MaterialExpressionNamedRerouteUsage_26}
    B {from=MaterialExpressionConstant2Vector_2}
    Alpha {from=MaterialExpressionNamedRerouteUsage_106}
   ]
   properties {desc="delete elements at their\r\noriginal location"}
  }
  MaterialExpressionLinearInterpolate_4 {
   class=MaterialExpressionLinearInterpolate
   inputs [
    A {from=MaterialExpressionLinearInterpolate_5}
    B {from=MaterialExpressionNamedRerouteUsage_88}
    Alpha {from=MaterialExpressionNamedRerouteUsage_89}
   ]
   properties {desc=add transformed left eye}
  }
  MaterialExpressionLinearInterpolate_5 {
   class=MaterialExpressionLinearInterpolate
   inputs [
    A {from=MaterialExpressionLinearInterpolate_1}
    B {from=MaterialExpressionNamedRerouteUsage_29}
    Alpha {from=MaterialExpressionNamedRerouteUsage_30}
   ]
   properties {desc=add transformed mouth}
  }
  MaterialExpressionLinearInterpolate_6 {
   class=MaterialExpressionLinearInterpolate
   inputs [
    A {from=MaterialExpressionLinearInterpolate_4}
    B {from=MaterialExpressionNamedRerouteUsage_68}
    Alpha {from=MaterialExpressionNamedRerouteUsage_69}
   ]
   properties {desc=add transformed right eye}
  }
  MaterialExpressionLocalPosition_1 {class=MaterialExpressionLocalPosition, inputs [], properties {localOrigin=InstancePreSkinning}}
  MaterialExpressionMaterialFunctionCall_11 {
   class=MaterialExpressionMaterialFunctionCall
   inputs [
    UV {from=MaterialExpressionNamedRerouteUsage_72}
    Element UV Center {from=MaterialExpressionNamedRerouteUsage_94}
    Element UV Radius {from=MaterialExpressionNamedRerouteUsage_95}
    Starting Pos UV Offset {from=MaterialExpressionNamedRerouteUsage_73}
    Starting Rotation Offset {from=MaterialExpressionNamedRerouteUsage_85}
    Starting Scale {from=MaterialExpressionNamedRerouteUsage_74}
    Translation UV Range {from=""}
    "Translation Timeline (Normalized)" {from=""}
    Rotation Timeline {from=""}
    Scale Timeline {from=""}
   ]
   properties {materialFunction=&/MF_TransformUVIsland_Disc}
  }
  MaterialExpressionMaterialFunctionCall_12 {
   class=MaterialExpressionMaterialFunctionCall
   inputs [
    UV {from=MaterialExpressionNamedRerouteUsage_83}
    Element UV Center {from=MaterialExpressionNamedRerouteUsage_96}
    Element UV Radius {from=MaterialExpressionNamedRerouteUsage_97}
    Starting Pos UV Offset {from=MaterialExpressionNamedRerouteUsage_79}
    Starting Rotation Offset {from=MaterialExpressionNamedRerouteUsage_86}
    Starting Scale {from=MaterialExpressionNamedRerouteUsage_75}
    Translation UV Range {from=""}
    "Translation Timeline (Normalized)" {from=""}
    Rotation Timeline {from=""}
    Scale Timeline {from=""}
   ]
   properties {materialFunction=&MF_TransformUVIsland_Disc}
  }
  MaterialExpressionMaterialFunctionCall_13 {
   class=MaterialExpressionMaterialFunctionCall
   inputs [
    UV {from=MaterialExpressionNamedRerouteUsage_84}
    Element UV Center {from=MaterialExpressionNamedRerouteUsage_98}
    Element UV Radius {from=MaterialExpressionNamedRerouteUsage_99}
    Starting Pos UV Offset {from=MaterialExpressionNamedRerouteUsage_82}
    Starting Rotation Offset {from=MaterialExpressionNamedRerouteUsage_87}
    Starting Scale {from=MaterialExpressionNamedRerouteUsage_76}
    Translation UV Range {from=""}
    "Translation Timeline (Normalized)" {from=""}
    Rotation Timeline {from=""}
    Scale Timeline {from=""}
   ]
   properties {materialFunction=&MF_TransformUVIsland_Disc}
  }
  MaterialExpressionMaterialFunctionCall_3 {
   class=MaterialExpressionMaterialFunctionCall
   inputs [
    RotationAmount {from=MaterialExpressionReroute_2}
    Coords {from=MaterialExpressionDivide_5}
    Pivot {from=MaterialExpressionConstant_2}
   ]
   properties {materialFunction=/BaseMaterial/Materials/Functions/MF_Rotate2D}
  }
  MaterialExpressionMax_3 {
   class=MaterialExpressionMax
   inputs [
    A {from=MaterialExpressionNamedRerouteUsage_25}
    B {from=MaterialExpressionNamedRerouteUsage_90}
   ]
  }
  MaterialExpressionMax_9 {
   class=MaterialExpressionMax
   inputs [
    A {from=MaterialExpressionMax_3}
    B {from=MaterialExpressionNamedRerouteUsage_24}
   ]
  }
  MaterialExpressionMultiply_0 {
   class=MaterialExpressionMultiply
   inputs [
    A {from=MaterialExpressionOneMinus_0}
    B {from=MaterialExpressionLinearInterpolate_0}
   ]
  }
  MaterialExpressionMultiply_4 {
   class=MaterialExpressionMultiply
   inputs [
    A {from=MaterialExpressionVectorParameter_2, output=Pivot Y}
    B {from=""}
   ]
   properties {constB=-1}
  }
  MaterialExpressionNamedRerouteDeclaration_0 {
   class=MaterialExpressionNamedRerouteDeclaration
   inputs [None {from=MaterialExpressionVertexInterpolator_2, output=PS}]
   properties {name="Base UV (V2)", nodeColor=<rgba 0.0522 1 0 1>}
  }
  MaterialExpressionNamedRerouteDeclaration_1 {
   class=MaterialExpressionNamedRerouteDeclaration
   inputs [None {from=MaterialExpressionMultiply_0}]
   properties {name="UV Projection Mask (S)", nodeColor=<rgba 0 1 0.3672 1>}
  }
  MaterialExpressionNamedRerouteDeclaration_12 {
   class=MaterialExpressionNamedRerouteDeclaration
   inputs [None {from=MaterialExpressionAppendVector_12}]
   properties {name="Mouth UV Offset (V2)", nodeColor=<rgba 0.2422 0 1 1>}
  }
  MaterialExpressionNamedRerouteDeclaration_13 {
   class=MaterialExpressionNamedRerouteDeclaration
   inputs [None {from=MaterialExpressionVectorParameter_14, output=Scale}]
   properties {name="Mouth Scale (S)", nodeColor=<rgba 0.6406 1 0 1>}
  }
  MaterialExpressionNamedRerouteDeclaration_14 {
   class=MaterialExpressionNamedRerouteDeclaration
   inputs [None {from=MaterialExpressionMaterialFunctionCall_11, output="Transformed Pos Mask (S)"}]
   properties {name="Right Eye Mask (S)", nodeColor=<rgba 1 0 0.4453 1>}
  }
  MaterialExpressionNamedRerouteDeclaration_15 {
   class=MaterialExpressionNamedRerouteDeclaration
   inputs [None {from=MaterialExpressionMaterialFunctionCall_11, output="Element UV (V2)"}]
   properties {name="Right Eye UV (V2)", nodeColor=<rgba 0 0.7422 1 1>}
  }
  MaterialExpressionNamedRerouteDeclaration_16 {
   class=MaterialExpressionNamedRerouteDeclaration
   inputs [None {from=MaterialExpressionMaterialFunctionCall_11, output="Orig Pos Mask (S)"}]
   properties {name="Right Eye Orig Mask (S)", nodeColor=<rgba 1 0 0.4453 1>}
  }
  MaterialExpressionNamedRerouteDeclaration_17 {
   class=MaterialExpressionNamedRerouteDeclaration
   inputs [None {from=MaterialExpressionMaterialFunctionCall_13, output="Transformed Pos Mask (S)"}]
   properties {name="Mouth Mask (S)", nodeColor=<rgba 1 0 0.4453 1>}
  }
  MaterialExpressionNamedRerouteDeclaration_18 {
   class=MaterialExpressionNamedRerouteDeclaration
   inputs [None {from=MaterialExpressionMaterialFunctionCall_13, output="Element UV (V2)"}]
   properties {name="Mouth UV (V2)", nodeColor=<rgba 0 0.7422 1 1>}
  }
  MaterialExpressionNamedRerouteDeclaration_19 {
   class=MaterialExpressionNamedRerouteDeclaration
   inputs [None {from=MaterialExpressionMaterialFunctionCall_13, output="Orig Pos Mask (S)"}]
   properties {name="Mouth Orig Mask (S)", nodeColor=<rgba 1 0 0.4453 1>}
  }
  MaterialExpressionNamedRerouteDeclaration_20 {
   class=MaterialExpressionNamedRerouteDeclaration
   inputs [None {from=MaterialExpressionVectorParameter_8, output=Rotation}]
   properties {name="Left Eye Rotation (S)", nodeColor=<rgba 0.6406 1 0 1>}
  }
  MaterialExpressionNamedRerouteDeclaration_21 {
   class=MaterialExpressionNamedRerouteDeclaration
   inputs [None {from=MaterialExpressionVectorParameter_14, output=Rotation}]
   properties {name="Mouth Rotation (S)", nodeColor=<rgba 0.6406 1 0 1>}
  }
  MaterialExpressionNamedRerouteDeclaration_22 {
   class=MaterialExpressionNamedRerouteDeclaration
   inputs [None {from=MaterialExpressionSaturate_6}]
   properties {name="Transformed UV (V2)", nodeColor=<rgba 0.0522 1 0 1>}
  }
  MaterialExpressionNamedRerouteDeclaration_24 {
   class=MaterialExpressionNamedRerouteDeclaration
   inputs [None {from=MaterialExpressionVectorParameter_9, output=Rotation}]
   properties {name="Right Eye Rotation (S)", nodeColor=<rgba 0.6406 1 0 1>}
  }
  MaterialExpressionNamedRerouteDeclaration_26 {
   class=MaterialExpressionNamedRerouteDeclaration
   inputs [None {from=MaterialExpressionAppendVector_15}]
   properties {name="Right Eye UV Center (V2)", nodeColor=<rgba 0 0.2969 1 1>}
  }
  MaterialExpressionNamedRerouteDeclaration_27 {
   class=MaterialExpressionNamedRerouteDeclaration
   inputs [None {from=MaterialExpressionVectorParameter_18, output=Radius}]
   properties {name="Right Eye UV Radius (S)", nodeColor=<rgba 0 0.2969 1 1>}
  }
  MaterialExpressionNamedRerouteDeclaration_32 {
   class=MaterialExpressionNamedRerouteDeclaration
   inputs [None {from=MaterialExpressionAppendVector_21}]
   properties {name="Left Eye UV Center (V2)", nodeColor=<rgba 0 0.2969 1 1>}
  }
  MaterialExpressionNamedRerouteDeclaration_33 {
   class=MaterialExpressionNamedRerouteDeclaration
   inputs [None {from=MaterialExpressionVectorParameter_23, output=Radius}]
   properties {name="Left Eye UV Radius (S)", nodeColor=<rgba 0 0.2969 1 1>}
  }
  MaterialExpressionNamedRerouteDeclaration_34 {
   class=MaterialExpressionNamedRerouteDeclaration
   inputs [None {from=MaterialExpressionAppendVector_22}]
   properties {name="Mouth UV Center (V2)", nodeColor=<rgba 0 0.2969 1 1>}
  }
  MaterialExpressionNamedRerouteDeclaration_35 {
   class=MaterialExpressionNamedRerouteDeclaration
   inputs [None {from=MaterialExpressionDDX_0}]
   properties {name="UV DDX (V2)", nodeColor=<rgba 0 0.6484 1 1>}
  }
  MaterialExpressionNamedRerouteDeclaration_36 {
   class=MaterialExpressionNamedRerouteDeclaration
   inputs [None {from=MaterialExpressionDDY_0}]
   properties {name="UV DDY (V2)", nodeColor=<rgba 0 0.6484 1 1>}
  }
  MaterialExpressionNamedRerouteDeclaration_37 {
   class=MaterialExpressionNamedRerouteDeclaration
   inputs [None {from=MaterialExpressionVectorParameter_24, output=Radius}]
   properties {name="Mouth UV Radius (S)", nodeColor=<rgba 0 0.2969 1 1>}
  }
  MaterialExpressionNamedRerouteDeclaration_4 {
   class=MaterialExpressionNamedRerouteDeclaration
   inputs [None {from=MaterialExpressionConstant_0}]
   properties {name="DEBUG (V3)", nodeColor=<rgba 1 0 0 1>}
  }
  MaterialExpressionNamedRerouteDeclaration_41 {
   class=MaterialExpressionNamedRerouteDeclaration
   inputs [None {from=MaterialExpressionAppendVector_11}]
   properties {name="Right Eye UV Offset (V2)", nodeColor=<rgba 0.2422 0 1 1>}
  }
  MaterialExpressionNamedRerouteDeclaration_5 {
   class=MaterialExpressionNamedRerouteDeclaration
   inputs [None {from=MaterialExpressionVectorParameter_9, output=Scale}]
   properties {name="Right Eye Scale (S)", nodeColor=<rgba 0.6406 1 0 1>}
  }
  MaterialExpressionNamedRerouteDeclaration_59 {
   class=MaterialExpressionNamedRerouteDeclaration
   inputs [None {from=MaterialExpressionMaterialFunctionCall_12, output="Transformed Pos Mask (S)"}]
   properties {name="Left Eye Mask (S)", nodeColor=<rgba 1 0 0.4453 1>}
  }
  MaterialExpressionNamedRerouteDeclaration_60 {
   class=MaterialExpressionNamedRerouteDeclaration
   inputs [None {from=MaterialExpressionMaterialFunctionCall_12, output="Element UV (V2)"}]
   properties {name="Left Eye UV (V2)", nodeColor=<rgba 0 0.7422 1 1>}
  }
  MaterialExpressionNamedRerouteDeclaration_61 {
   class=MaterialExpressionNamedRerouteDeclaration
   inputs [None {from=MaterialExpressionMaterialFunctionCall_12, output="Orig Pos Mask (S)"}]
   properties {name="Left Eye Orig Mask (S)", nodeColor=<rgba 1 0 0.4453 1>}
  }
  MaterialExpressionNamedRerouteDeclaration_71 {
   class=MaterialExpressionNamedRerouteDeclaration
   inputs [None {from=MaterialExpressionMax_9}]
   properties {name="Face Orig Mask (S)", nodeColor=<rgba 1 0 0.4922 1>}
  }
  MaterialExpressionNamedRerouteDeclaration_8 {
   class=MaterialExpressionNamedRerouteDeclaration
   inputs [None {from=MaterialExpressionAppendVector_8}]
   properties {name="Left Eye UV Offset (V2)", nodeColor=<rgba 0.2422 0 1 1>}
  }
  MaterialExpressionNamedRerouteDeclaration_9 {
   class=MaterialExpressionNamedRerouteDeclaration
   inputs [None {from=MaterialExpressionVectorParameter_8, output=Scale}]
   properties {name="Left Eye Scale (S)", nodeColor=<rgba 0.6406 1 0 1>}
  }
  MaterialExpressionNamedRerouteUsage_0 {class=MaterialExpressionNamedRerouteUsage}
  MaterialExpressionNamedRerouteUsage_1 {class=MaterialExpressionNamedRerouteUsage}
  MaterialExpressionNamedRerouteUsage_10 {class=MaterialExpressionNamedRerouteUsage}
  MaterialExpressionNamedRerouteUsage_106 {class=MaterialExpressionNamedRerouteUsage}
  MaterialExpressionNamedRerouteUsage_11 {class=MaterialExpressionNamedRerouteUsage}
  MaterialExpressionNamedRerouteUsage_2 {class=MaterialExpressionNamedRerouteUsage}
  MaterialExpressionNamedRerouteUsage_24 {class=MaterialExpressionNamedRerouteUsage}
  MaterialExpressionNamedRerouteUsage_25 {class=MaterialExpressionNamedRerouteUsage}
  MaterialExpressionNamedRerouteUsage_26 {class=MaterialExpressionNamedRerouteUsage}
  MaterialExpressionNamedRerouteUsage_29 {class=MaterialExpressionNamedRerouteUsage}
  MaterialExpressionNamedRerouteUsage_3 {class=MaterialExpressionNamedRerouteUsage}
  MaterialExpressionNamedRerouteUsage_30 {class=MaterialExpressionNamedRerouteUsage}
  MaterialExpressionNamedRerouteUsage_4 {class=MaterialExpressionNamedRerouteUsage}
  MaterialExpressionNamedRerouteUsage_5 {class=MaterialExpressionNamedRerouteUsage}
  MaterialExpressionNamedRerouteUsage_6 {class=MaterialExpressionNamedRerouteUsage}
  MaterialExpressionNamedRerouteUsage_68 {class=MaterialExpressionNamedRerouteUsage}
  MaterialExpressionNamedRerouteUsage_69 {class=MaterialExpressionNamedRerouteUsage}
  MaterialExpressionNamedRerouteUsage_7 {class=MaterialExpressionNamedRerouteUsage}
  MaterialExpressionNamedRerouteUsage_72 {class=MaterialExpressionNamedRerouteUsage}
  MaterialExpressionNamedRerouteUsage_73 {class=MaterialExpressionNamedRerouteUsage}
  MaterialExpressionNamedRerouteUsage_74 {class=MaterialExpressionNamedRerouteUsage}
  MaterialExpressionNamedRerouteUsage_75 {class=MaterialExpressionNamedRerouteUsage}
  MaterialExpressionNamedRerouteUsage_76 {class=MaterialExpressionNamedRerouteUsage}
  MaterialExpressionNamedRerouteUsage_79 {class=MaterialExpressionNamedRerouteUsage}
  MaterialExpressionNamedRerouteUsage_8 {class=MaterialExpressionNamedRerouteUsage}
  MaterialExpressionNamedRerouteUsage_82 {class=MaterialExpressionNamedRerouteUsage}
  MaterialExpressionNamedRerouteUsage_83 {class=MaterialExpressionNamedRerouteUsage}
  MaterialExpressionNamedRerouteUsage_84 {class=MaterialExpressionNamedRerouteUsage}
  MaterialExpressionNamedRerouteUsage_85 {class=MaterialExpressionNamedRerouteUsage}
  MaterialExpressionNamedRerouteUsage_86 {class=MaterialExpressionNamedRerouteUsage}
  MaterialExpressionNamedRerouteUsage_87 {class=MaterialExpressionNamedRerouteUsage}
  MaterialExpressionNamedRerouteUsage_88 {class=MaterialExpressionNamedRerouteUsage}
  MaterialExpressionNamedRerouteUsage_89 {class=MaterialExpressionNamedRerouteUsage}
  MaterialExpressionNamedRerouteUsage_9 {class=MaterialExpressionNamedRerouteUsage}
  MaterialExpressionNamedRerouteUsage_90 {class=MaterialExpressionNamedRerouteUsage}
  MaterialExpressionNamedRerouteUsage_94 {class=MaterialExpressionNamedRerouteUsage}
  MaterialExpressionNamedRerouteUsage_95 {class=MaterialExpressionNamedRerouteUsage}
  MaterialExpressionNamedRerouteUsage_96 {class=MaterialExpressionNamedRerouteUsage}
  MaterialExpressionNamedRerouteUsage_97 {class=MaterialExpressionNamedRerouteUsage}
  MaterialExpressionNamedRerouteUsage_98 {class=MaterialExpressionNamedRerouteUsage}
  MaterialExpressionNamedRerouteUsage_99 {class=MaterialExpressionNamedRerouteUsage}
  MaterialExpressionOneMinus_0 {class=MaterialExpressionOneMinus, inputs [None {from=MaterialExpressionStep_0}]}
  MaterialExpressionPreSkinnedNormal_0 {class=MaterialExpressionPreSkinnedNormal, inputs [], properties {}}
  MaterialExpressionReroute_0 {class=MaterialExpressionReroute, inputs [None {from=MaterialExpressionVectorParameter_2, output=Size}]}
  MaterialExpressionReroute_1 {class=MaterialExpressionReroute, inputs [None {from=MaterialExpressionVectorParameter_0, output=Mask Mix}]}
  MaterialExpressionReroute_2 {class=MaterialExpressionReroute, inputs [None {from=MaterialExpressionVectorParameter_2, output=Rotation}]}
  MaterialExpressionSaturate_6 {class=MaterialExpressionSaturate, inputs [None {from=MaterialExpressionLinearInterpolate_6}], properties {desc=prevent texture wrap}}
  MaterialExpressionSmoothStep_0 {
   class=MaterialExpressionSmoothStep
   inputs [
    Min {from=MaterialExpressionVectorParameter_0, output=PreSkinned Normal Z Min}
    Max {from=MaterialExpressionVectorParameter_0, output=PreSkinned Normal Z Min}
    Value {from=MaterialExpressionComponentMask_0, output=""}
   ]
   properties {}
  }
  MaterialExpressionStep_0 {
   class=MaterialExpressionStep
   inputs [
    Y {from="", output=""}
    X {from=MaterialExpressionDivide_1, output=""}
   ]
   properties {constY=1}
  }
  MaterialExpressionSubtract_0 {
   class=MaterialExpressionSubtract
   inputs [
    A {from=MaterialExpressionTextureCoordinate_1}
    B {from=MaterialExpressionAppendVector_2}
   ]
  }
  MaterialExpressionSubtract_8 {
   class=MaterialExpressionSubtract
   inputs [
    A {from=MaterialExpressionLocalPosition_1, output=XY}
    B {from=MaterialExpressionAppendVector_3}
   ]
  }
  MaterialExpressionTextureCoordinate_1 {class=MaterialExpressionTextureCoordinate}
  MaterialExpressionTextureCoordinate_3 {class=MaterialExpressionTextureCoordinate}
  MaterialExpressionVectorParameter_0 {
   class=MaterialExpressionVectorParameter
   properties {
    group=Face Decals | Setup
    parameterName=Decals Projection Up-Facing Mask
    sortPriority=20
    channelNames {a=n/a, b=Mask Mix, g=PreSkinned Normal Z Max, r=PreSkinned Normal Z Min}
    defaultValue=<rgba 0.1 0.4 1 0>
   }
  }
  MaterialExpressionVectorParameter_1 {
   class=MaterialExpressionVectorParameter
   properties {
    group=Face Decals | Setup
    parameterName=Decals Projection UV Mask
    sortPriority=10
    channelNames {a=n/a, b=Radius, g=Center V, r=Center U}
    defaultValue=<rgba 0.584 0.603 0.38 0>
   }
  }
  MaterialExpressionVectorParameter_14 {
   class=MaterialExpressionVectorParameter
   properties {
    group=Face Decals | Transforms
    parameterName=Mouth UV Transforms
    sortPriority=320
    channelNames {a=Rotation, b=Scale, g=Offset V, r=Offset U}
    defaultValue=<rgba 0 0 1 0>
   }
  }
  MaterialExpressionVectorParameter_18 {
   class=MaterialExpressionVectorParameter
   properties {
    group=Face Decals | Setup
    parameterName=Right Eye UV Island
    sortPriority=110
    channelNames {a=n/a, b=Radius, g=Center V, r=Center U}
    defaultValue=<rgba 0.193 0.5 0.13 0>
   }
  }
  MaterialExpressionVectorParameter_2 {
   class=MaterialExpressionVectorParameter
   properties {
    desc=In pre-skinned local space
    group=Face Decals | Transforms
    parameterName=Decals Base Transforms
    sortPriority=10
    channelNames {a=Rotation, b=Size, g=Pivot Y, r=Pivot X}
    defaultValue=<rgba 0 0 90 0>
   }
  }
  MaterialExpressionVectorParameter_23 {
   class=MaterialExpressionVectorParameter
   properties {
    group=Face Decals | Setup
    parameterName=Left Eye UV Island
    sortPriority=210
    channelNames {a=n/a, b=Radius, g=Center V, r=Center U}
    defaultValue=<rgba 0.804 0.5 0.13 0>
   }
  }
  MaterialExpressionVectorParameter_24 {
   class=MaterialExpressionVectorParameter
   properties {
    group=Face Decals | Setup
    parameterName=Mouth UV Island
    sortPriority=310
    channelNames {a=n/a, b=Radius, g=Center V, r=Center U}
    defaultValue=<rgba 0.505 0.665 0.08 0>
   }
  }
  MaterialExpressionVectorParameter_8 {
   class=MaterialExpressionVectorParameter
   properties {
    group=Face Decals | Transforms
    parameterName=Left Eye UV Transforms
    sortPriority=220
    channelNames {a=Rotation, b=Scale, g=Offset V, r=Offset U}
    defaultValue=<rgba 0 0 1 0>
   }
  }
  MaterialExpressionVectorParameter_9 {
   class=MaterialExpressionVectorParameter
   properties {
    group=Face Decals | Transforms
    parameterName=Right Eye UV Transforms
    sortPriority=120
    channelNames {a=Rotation, b=Scale, g=Offset V, r=Offset U}
    defaultValue=<rgba 0 0 1 0>
   }
  }
  MaterialExpressionVertexInterpolator_0 {class=MaterialExpressionVertexInterpolator, inputs [VS {from=MaterialExpressionPreSkinnedNormal_0, output=""}], properties {}}
  MaterialExpressionVertexInterpolator_2 {class=MaterialExpressionVertexInterpolator, inputs [VS {from=MaterialExpressionAdd_4, output=""}], properties {}}
 }}
