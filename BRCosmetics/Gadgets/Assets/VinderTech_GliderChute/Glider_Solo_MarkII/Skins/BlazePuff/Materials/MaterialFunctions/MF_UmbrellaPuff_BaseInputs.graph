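# Material function supplying base surface and face-decal inputs for the BlazePuff umbrella glider skin.
# Surface attributes come from color, SRM and normal textures, with emissive scaled by a time-of-day tint.
# MF_UmbrellaPuff_FaceUV provides projected UVs and derivatives for explicit-gradient sampling of face color and mask, with contrast-adjusted color driving an emissive-only face material.
# Separate outputs expose both materials, surface normal, masks, face UVs and debug.
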MaterialFunction /BRCosmetics/Gadgets/Assets/VinderTech_GliderChute/Glider_Solo_MarkII/Skins/BlazePuff/Materials/MaterialFunctions/MF_UmbrellaPuff_BaseInputs
graph {nodes [
  n1 = TextureSampleParameter2D!(UVs=NamedRerouteUsage(), "DDX(UVs)"=NamedRerouteUsage(), "DDY(UVs)"=NamedRerouteUsage(), Apply View MipBias) {
   group=Umbrella|Textures
   mipValueMode=TMVM_Derivative
   parameterName=FaceMask
   samplerType=SAMPLERTYPE_Masks
   sortPriority=10
   texture=/Game/Packages/Fortress_SharedMaterials/Textures/Black_MASK
  }
  n2 = Max(A=@n1.RGB) {constB=0.1}
  n3 = NamedRerouteUsage()
  DotProduct(A=@n2, B=@n3)
  n4 = TextureSampleParameter2D!(UVs, Apply View MipBias) {
   group=Umbrella|Textures
   parameterName=Base Color
   samplerSource=SSM_Wrap_WorldGroupSettings
   sortPriority=40
   texture=&T_BlacksRGB
  }
  n5 = TextureSampleParameter2D() {
   group=Umbrella|Textures
   parameterName=SRM
   samplerSource=SSM_Wrap_WorldGroupSettings
   samplerType=SAMPLERTYPE_Masks
   sortPriority=50
   texture=&T_BlackMask
  }
  n6 = MaterialFunctionCall(
   Color_Morning=VectorParameter() {group=Sprite_Base, parameterName=TOD_Morning, sortPriority=9, defaultValue=<rgba 0.8 0.4 0.5 1>}.RGB
   Color_Day=Constant()
   Color_Evening=VectorParameter() {group=Sprite_Base, parameterName=TOD_Evening, sortPriority=9, defaultValue=<rgba 0.8 0.5 0.4 1>}.RGB
   Color_Night=VectorParameter() {group=Sprite_Base, parameterName=TOD_Night, sortPriority=9, defaultValue=<rgba 0.35 0.4667 0.7 1>}.RGB
  ) {materialFunction=&Color_Time_of_Day}
  n7 = ScalarParameter() {defaultValue=0.5, parameterName=BaseEmissive Tex ToD Intensity}
  n8 = TextureSampleParameter2D() {
   group=Umbrella|Textures
   parameterName=Normal
   samplerSource=SSM_Wrap_WorldGroupSettings
   samplerType=SAMPLERTYPE_Normal
   sortPriority=60
   texture=&FlatNormalBlue
  }
  n9 = SetMaterialAttributes!(
   MaterialAttributes
   BaseColor=NamedRerouteDeclaration(None=@n4.RGB) {name="BaseColor (V3)", nodeColor=<rgba 1 0 0.1641 1>}
   Metallic=@n5.B
   Specular=@n5.B
   Roughness=@n5.B
   Anisotropy
   EmissiveColor=Multiply(A=NamedRerouteUsage(), B=Multiply(A=@n6.Result, B=@n7) {constB=0.5})
   ShadingModel=ShadingModel()
   OpacityMask
   Normal=@n8.RGB
  ) {attributeSetTypes [
    #1
    #2
    #3
    #4
    #5
    #6
    #7
    #8
    #9
   ]}
  FunctionOutput(None=@n9) {outputName="Surface Material (MA)"}
  FunctionOutput(None=@n8.RGB) {outputName="Surface Normal TS (V3)", sortPriority=10}
  n10 = TextureSampleParameter2D() {
   group=Umbrella|Textures
   parameterName=PaintMask
   samplerSource=SSM_Wrap_WorldGroupSettings
   samplerType=SAMPLERTYPE_Masks
   sortPriority=30
   texture=&T_BlackMask
  }
  FunctionOutput(None=@n10.R) {outputName="Surface Mask (S)", sortPriority=20}
  FunctionOutput(None=NamedRerouteUsage()) {outputName="DEBUG (V3)", sortPriority=1000}
  n11 = TextureSampleParameter2D!(UVs=NamedRerouteUsage(), "DDX(UVs)"=NamedRerouteUsage(), "DDY(UVs)"=NamedRerouteUsage(), Apply View MipBias) {
   group=Umbrella|Textures
   mipValueMode=TMVM_Derivative
   parameterName=FaceMask
   samplerSource=SSM_Clamp_WorldGroupSettings
   samplerType=SAMPLERTYPE_Masks
   sortPriority=10
   texture=&Black_MASK
  }
  FunctionOutput(None=Multiply(A=@n11.R, B=NamedRerouteUsage()) {constA=1}) {outputName="Face Mask (S)", sortPriority=110}
  n12 = TextureSampleParameter2D(UVs=NamedRerouteUsage(), "DDX(UVs)"=NamedRerouteUsage(), "DDY(UVs)"=NamedRerouteUsage()) {
   group=Umbrella|Textures
   mipValueMode=TMVM_Derivative
   parameterName=Face Color
   samplerSource=SSM_Clamp_WorldGroupSettings
   sortPriority=20
   texture=/Game/Characters/Fortress_Character/Textures/T_All_White
  }
  n13 = MaterialFunctionCall!(In=@n12.RGB, Contrast=ScalarParameter() {group=Face Decals | Remap, parameterName=Face BaseColor Contrast, sortPriority=10}) {materialFunction=/Engine/Functions/Engine_MaterialFunctions01/ImageAdjustment/CheapContrast_RGB}
  n14 = ScalarParameter() {defaultValue=1, group=Face Decals | Remap, parameterName=Face BaseColor Mult, sortPriority=20}
  n15 = Constant()
  n16 = SetMaterialAttributes!(
   MaterialAttributes
   EmissiveColor=Saturate(None=Multiply(A=@n13.Result, B=@n14))
   BaseColor=@n15
   Metallic=@n15
   Specular=@n15
   Roughness=@n15
   Normal=Convert!() {convertOutputs [{type=Vector3, defaultValue=<rgba 0 0 1 1>}]}
   OpacityMask=Constant() {r=1}
   ShadingModel=ShadingModel()
  ) {attributeSetTypes [
    #6
    #1
    #2
    #3
    #4
    #9
    #8
    #7
   ]}
  FunctionOutput(None=@n16) {outputName="Face Material (MA)", sortPriority=100}
  FunctionOutput(None=NamedRerouteUsage()) {outputName="Face UVs (V2)", sortPriority=120}
  n17 = MaterialFunctionCall() {materialFunction=&/MF_UmbrellaPuff_FaceUV}
  NamedRerouteDeclaration(None=@n17."Base Projected UV (V2)") {name="Face Base UV (V2)", nodeColor=<rgba 0.0522 1 0 1>}
  NamedRerouteDeclaration(None=@n17."Projection Mask (S)") {name="Face Projection Mask (S)", nodeColor=<rgba 0 1 0.3672 1>}
  NamedRerouteDeclaration(None=@n17."Orig Pos Mask (V3)") {name="Face Orig Pos Mask (V3)", nodeColor=<rgba 1 0 0.6783 1>}
  NamedRerouteDeclaration(None=@n17."Transformed Projected UV (V2)") {name="Face Transformed UV (V2)", nodeColor=<rgba 0.0522 1 0 1>}
  NamedRerouteDeclaration(None=@n17."Transformed Pos Mask (V3)") {name="Face Transformed Pos Mask (V3)", nodeColor=<rgba 1 0 0.6783 1>}
  NamedRerouteDeclaration(None=@n17."UV DDX (V2)") {name="Face UV DDX (V2)", nodeColor=<rgba 0 0.6484 1 1>}
  NamedRerouteDeclaration(None=@n17."UV DDY (V2)") {name="Face UV DDY (V2)", nodeColor=<rgba 0 0.6484 1 1>}
  NamedRerouteDeclaration(None=@n17."DEBUG (V3)") {name="DEBUG (V3)", nodeColor=<rgba 1 0 0 1>}
  NamedRerouteUsage()
  Saturate(None=Add(A=@n2, B=@n3))
  VectorParameter() {group=Sprite_Base, parameterName=TOD_Day, sortPriority=9, defaultValue=<rgba 1 1 1 1>}
 ]}
